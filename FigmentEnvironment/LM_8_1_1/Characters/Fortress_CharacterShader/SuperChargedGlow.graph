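# Material function for a character shader that adds a flickering, blue-tinted supercharged rim glow to the incoming Emissive.
# Fresnel is multiplied by a steep falloff around AverageHandPosition, which the UseDistanceCalc static switch can bypass.
# That mask tints blue and its boosted square adds a hot core, modulated by time-scrolled noise plus a flicker bias and scaled by SuperChargedEffect.
# An actor-distance sine pulse and a TransformPosition node are left unconnected.
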MaterialFunction /FigmentEnvironment/LM_8_1_1/Characters/Fortress_CharacterShader/SuperChargedGlow
graph {nodes [
  n1 = Multiply(A=Distance(A=ActorPositionWS(), B=WorldPosition().XYZ)) {constB=0.0078}
  n2 = MaterialFunctionCall(Speed=Constant() {r=-2}) {materialFunction=&TimeWithSpeedVariable}
  ConstantBiasScale(None=Sine(None=Add(A=@n1, B=@n2.Frac Time))) {bias=2}
  n3 = Constant3Vector() {constant=<rgba 0.1451 0.6275 0.9882 0>}
  n4 = Distance(A=FunctionInput() {bUsePreviewValueAsDefault=true, inputName=AverageHandPosition}, B=WorldPosition().XYZ)
  n5 = OneMinus(None=Clamp(None=Divide(A=@n4) {constB=90}))
  n6 = Multiply(A=@n5, B=@n5)
  n7 = FunctionInput(Preview=StaticBool() {value=true}) {bUsePreviewValueAsDefault=true, inputName=UseDistanceCalc, inputType=FunctionInput_StaticBool, previewValue=<1 0 0 1>}
  n8 = Multiply(
   A=Fresnel(Normal=PixelNormalWS()) {baseReflectFraction=0, exponent=1}
   B=StaticSwitch(True=Multiply(A=@n6, B=@n6), False=Constant() {r=1}, Value=@n7)
  )
  n9 = Multiply(A=@n8, B=@n8)
  n10 = Multiply(A=Time(), B=Constant2Vector() {g=0.025, r=0.01}) {constB=0.01}
  n11 = FunctionInput() {bUsePreviewValueAsDefault=true, inputName=SuperChargedFlickerBias, inputType=FunctionInput_Scalar, previewValue=<1 0 0 1>}
  n12 = Multiply(
   A=Add(A=Multiply(A=@n3, B=@n8), B=Multiply(A=@n9) {constB=3})
   B=Add(A=TextureSample(UVs=@n10) {texture=/FigmentEnvironment/LM_8_1_1/UI/UI_Faces/Noise}.R, B=@n11)
  )
  n13 = FunctionInput() {bUsePreviewValueAsDefault=true, inputName=SuperChargedEffect}
  FunctionOutput(None=Add(A=Multiply(A=@n12, B=@n13), B=FunctionInput() {bUsePreviewValueAsDefault=true, inputName=Emissive}))
  TransformPosition(None=Constant3Vector() {constant=<rgba 32 -32 110 0>}) {transformType=TRANSFORMPOSSOURCE_World}
 ]}
